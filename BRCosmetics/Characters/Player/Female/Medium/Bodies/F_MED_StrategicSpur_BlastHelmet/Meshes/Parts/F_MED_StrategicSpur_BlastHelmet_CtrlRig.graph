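# Control Rig for the female StrategicSpur BlastHelmet part that applies pose-driven corrective offsets to helmet hair bones, including 05 and 06, as the head rotates.
# Three spherical pose readers on the head bone weight four chained Modify Transforms steps.
# The first reader is scaled by NotCrouch×movingCtrl and separately by PistolMove, the second drives bone 06 ungated, and the third is scaled by CrouchStatus.
# A quaternion-to-Euler round trip feeds nothing.
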ControlRigBlueprint /BRCosmetics/Characters/Player/Female/Medium/Bodies/F_MED_StrategicSpur_BlastHelmet/Meshes/Parts/F_MED_StrategicSpur_BlastHelmet_CtrlRig
graphs [
 Multiply_1_1_ContainedGraph {nodes [
   n1 = RigVMFunctionEntryNode()
   RigVMFunctionReturnNode(Result=&RigVMFunction_MathDoubleMul(B=@n1.C, A=&RigVMFunction_MathDoubleMul(A=@n1.A, B=@n1.B).Result).Result)
  ]}
 RigVMFunctionLibrary {links [], nodes []}
 RigVMModel {nodes [
   RigUnit_BeginExecution = &RigUnit_BeginExecution()
   n1 = &RigUnit_SphericalPoseReader(
    ActiveRegionScaleFactors=(PositiveWidth=1.5,NegativeWidth=0.15,PositiveHeight=2,NegativeHeight=2)
    ActiveRegionSize=0.1
    Debug=(bDrawDebug=true,bDraw2D=false,bDrawLocalAxes=False,DebugScale=25,DebugSegments=20,DebugThickness=0.25)
    DriverAxis=(X=0,Y=1,Z=0)
    DriverItem=(Type=Bone,Name="head")
    FalloffRegionScaleFactors=(PositiveWidth=1,NegativeWidth=0.45,PositiveHeight=1,NegativeHeight=1)
    FalloffSize=0.3
    FlipHeightScaling=False
    FlipWidthScaling=False
    OptionalParentItem=(Type=Bone,Name="None")
    RotationOffset=(X=80,Y=0,Z=-23)
    ExecutePin=@RigUnit_BeginExecution.ExecutePin
   )
   Multiply_1 = RigVMAggregateNode(B=RigVMVariableNode() {variable=NotCrouch}.Value, C=RigVMVariableNode() {variable=movingCtrl}.Value, A=@n1.OutputParam)
   > &RigUnit_ModifyTransforms(
    ItemToModify="((Item=(Type=Bone,Name=\"base_helmetHair_05_m\"),Transform=(Rotation=(X=0,Y=-0.2341,Z=0,W=0.9722),Translation=(X=0.2203,Y=0,Z=2.1211),Scale3D=(X=1,Y=1,Z=1))),(Item=(Type=Bone,Name=\"base_helmetHair_06_m\"),Transform=(Rotation=(X=-0.1723,Y=-0.504,Z=-0.0437,W=0. ...(truncated)"
    WeightMinimum=0
    Weight=@Multiply_1.Result
   )
   > &RigUnit_ModifyTransforms(
    ItemToModify=((Item=(Type=Bone,Name="base_helmetHair_06_m"),Transform=(Rotation=(X=-0.0244,Y=-0.0599,Z=-0.0989,W=0.993),Translation=(X=-0.0469,Y=0,Z=0.2369),Scale3D=(X=1,Y=1,Z=1))))
    WeightMinimum=0
    Weight=&RigVMFunction_MathDoubleMul(B=RigVMVariableNode() {variable=PistolMove}.Value, A=@n1.OutputParam).Result
   )
   > n2 = &RigUnit_SphericalPoseReader(
    ActiveRegionScaleFactors=(PositiveWidth=0.2,NegativeWidth=0.2,PositiveHeight=0.5,NegativeHeight=0.3)
    ActiveRegionSize=0.1
    Debug=(bDrawDebug=true,bDraw2D=False,bDrawLocalAxes=False,DebugScale=25,DebugSegments=20,DebugThickness=0.25)
    DriverAxis=(X=0,Y=1,Z=0)
    DriverItem=(Type=Bone,Name="head")
    FalloffRegionScaleFactors=(PositiveWidth=0.3,NegativeWidth=0.3,PositiveHeight=0.2,NegativeHeight=0.2)
    FalloffSize=0.3
    FlipHeightScaling=False
    FlipWidthScaling=False
    OptionalParentItem=(Type=Bone,Name="None")
    RotationOffset=(X=115,Y=0,Z=-2)
   )
   > &RigUnit_ModifyTransforms(
    ItemToModify=((Item=(Type=Bone,Name="base_helmetHair_06_m"),Transform=(Rotation=(X=-0.0674,Y=-0.2298,Z=-0.0153,W=0.9708),Translation=(X=0,Y=0,Z=0),Scale3D=(X=1,Y=1,Z=1))))
    WeightMinimum=0
    Weight=@n2.OutputParam
   )
   > n3 = &RigUnit_SphericalPoseReader(
    ActiveRegionScaleFactors=(PositiveWidth=1.5,NegativeWidth=0.25,PositiveHeight=2,NegativeHeight=2)
    ActiveRegionSize=0.1
    Debug=(bDrawDebug=false,bDraw2D=False,bDrawLocalAxes=False,DebugScale=25,DebugSegments=20,DebugThickness=0.25)
    DriverAxis=(X=0,Y=1,Z=0)
    DriverItem=(Type=Bone,Name="head")
    FalloffRegionScaleFactors=(PositiveWidth=1,NegativeWidth=0.45,PositiveHeight=1,NegativeHeight=1)
    FalloffSize=0.3
    FlipHeightScaling=False
    FlipWidthScaling=False
    OptionalParentItem=(Type=Bone,Name="None")
    RotationOffset=(X=85,Y=0,Z=-32)
   )
   Multiply = &RigVMFunction_MathDoubleMul(A=@n3.OutputParam, B=RigVMVariableNode() {variable=CrouchStatus}.Value)
   > &RigUnit_ModifyTransforms(
    ItemToModify="((Item=(Type=Bone,Name=\"base_helmetHair_05_m\"),Transform=(Rotation=(X=0,Y=-0.4394,Z=0,W=0.8983),Translation=(X=-1.6607,Y=0,Z=0.6294),Scale3D=(X=1,Y=1,Z=1))),(Item=(Type=Bone,Name=\"base_helmetHair_06_m\"),Transform=(Rotation=(X=-0.0949,Y=-0.6685,Z=-0.0771,W= ...(truncated)"
    WeightMinimum=0
    Weight=@Multiply.Result
   )
   RigVMFunction_MathQuaternionToEuler = &RigVMFunction_MathQuaternionToEuler(RotationOrder=YZX, Value=(X=-0.1723,Y=-0.504,Z=-0.0437,W=0.8452))
   RigVMFunction_MathQuaternionFromEuler = &RigVMFunction_MathQuaternionFromEuler(
    Euler=(X=0,Y=60.1138,Z=11.563)
    RotationOrder=ZXY
    Euler.X=&RigVMFunction_MathDoubleMul(A=@RigVMFunction_MathQuaternionToEuler.Result.X, B=RigVMVariableNode() {variable=PistolMove}.Value).Result
    Euler.Y=@RigVMFunction_MathQuaternionToEuler.Result.Y
    Euler.Z=@RigVMFunction_MathQuaternionToEuler.Result.Z
   )
  ]}
]
hierarchy [
 root {loc=<0 0 0>, rot=<0 0 0 1>}
 pelvis {parent=root, loc=<0 -0.5301 98.0571>, rot=<0.0127 0.707 -0.0127 -0.707>}
 spine_01 {parent=pelvis, loc=<6.1208 -0.9328 0>, rot=<0 0 0.034 -0.9994>}
 spine_02 {parent=spine_01, loc=<6.5996 -0.2095 0>, rot=<0 0 0.0238 -0.9997>}
 spine_03 {parent=spine_02, loc=<6.4914 -0.1237 0>, rot=<0 0 0.0149 -0.9999>}
 spine_04 {parent=spine_03, loc=<6.9598 0.0794 0>, rot=<0 0 -0.0132 -0.9999>}
 spine_05 {parent=spine_04, loc=<6.9987 0 0>, rot=<0 0 -0.0015 -1>}
 clavicle_l {parent=spine_05, loc=<5.4831 -0.1053 -3.2946>, rot=<0.0465 -0.7289 -0.009 -0.6829>}
 neck_01 {parent=spine_05, loc=<11.3138 0.1529 0>, rot=<0 0 -0.2003 -0.9797>}
 neck_02 {parent=neck_01, loc=<4.5627 0 0>, rot=<0 0 0.0504 -0.9987>}
 head {parent=neck_02, loc=<5.3914 0 0>, rot=<0 0 0.115 -0.9934>}
 base_helmetHair_01_m {parent=head, loc=<13.7775 6.7992 0.2739>, rot=<-0.6823 -0.1858 -0.1858 -0.6823>}
 dyn_helmetHair_01_m {parent=base_helmetHair_01_m, loc=<0 0 0>, rot=<0 0 0 1>}
 base_helmetHair_02_m {parent=head, loc=<14.9832 1.0756 0.2739>, rot=<-0.7063 -0.0338 -0.0338 -0.7063>}
 dyn_helmetHair_02_m {parent=base_helmetHair_02_m, loc=<0 0 0>, rot=<0 0 0 1>}
 base_helmetHair_03_m {parent=head, loc=<13.4964 -5.4365 -0.099>, rot=<0.6568 -0.2619 0.2619 -0.6568>}
 dyn_helmetHair_03_m {parent=base_helmetHair_03_m, loc=<0 0 0>, rot=<0 0 0 1>}
 base_helmetHair_04_m {parent=head, loc=<7.3745 -8.5409 -0.2397>, rot=<0.5021 -0.4979 0.4979 -0.5021>}
 dyn_helmetHair_04_m {parent=base_helmetHair_04_m, loc=<0 0 0>, rot=<0 0 0 1>}
 base_helmetHair_05_m {parent=head, loc=<1.4763 -7.9845 0.0621>, rot=<0.4568 -0.5398 0.5398 -0.4568>}
 dyn_helmetHair_05_m {parent=base_helmetHair_05_m, loc=<0 0 0>, rot=<0 0 0 1>}
 base_helmetHair_06_m {parent=head, loc=<-1.5069 -7.4777 -0.1215>, rot=<0.3015 -0.6396 0.6396 -0.3015>}
 dyn_helmetHair_06_m {parent=base_helmetHair_06_m, loc=<0 0 0>, rot=<0 0 0 1>}
 clavicle_r {parent=spine_05, loc=<5.4831 -0.1053 3.2946>, rot=<0.7289 0.0465 -0.6829 0.009>}
]
variables [
 CrouchStatus {type=double, default=0}
 NotCrouch {type=double, default=0}
 PistolMove {type=double, default=0}
 movingCtrl {type=double, default=0}
]
