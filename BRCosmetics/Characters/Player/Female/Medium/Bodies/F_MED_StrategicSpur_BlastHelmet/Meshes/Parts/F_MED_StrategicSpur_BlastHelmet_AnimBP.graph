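# Post-copy AnimBP for the StrategicSpur BlastHelmet female body part, adding secondary motion to its helmet hair.
# Copies the attached parent's pose into a Control Rig fed by crouch, movement and pistol-style weapon alphas.
# Seven chained root-relative AnimDynamics spring nodes then simulate the hair bones, before an additive pitch offset on base_helmetHair_06_m.
# EventGraph also computes gravity, zipline and skydive-float values, which the AnimGraph never visibly reads.
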AnimBlueprint /BRCosmetics/Characters/Player/Female/Medium/Bodies/F_MED_StrategicSpur_BlastHelmet/Meshes/Parts/F_MED_StrategicSpur_BlastHelmet_AnimBP
target_skeleton=&/F_MED_StrategicSpur_BlastHelmet_Skeleton
graphs [
 AnimGraph {nodes [
   n1 = "Transform (Modify) Bone - Bone: base_helmetHair_06_m"!(
    ComponentPose=Local To Component(LocalPose=Use cached pose 'CRCache_Dyn'())
    bAlphaBoolEnabled:bool=True
    Translation="0,0,0"
    Rotation="32,0,0"
    Scale="1,1,1"
   ) -> (Pose) {
    alphaInputType=Bool
    rotationMode=BMM_Additive
    rotationSpace=BCS_BoneSpace
    translationMode=BMM_Additive
    translationSpace=BCS_BoneSpace
    alphaBoolBlend {blendInTime=0.5, blendOption=Linear, blendOutTime=0.5}
    alphaScaleBiasClamp {bInterpResult=true}
    boneToModify {boneName=base_helmetHair_06_m}
    rotation {pitch=32}
   }
   Output Pose(Result=Component To Local(ComponentPose=@n1)) {layerGroup=None}
   n2 = Control Rig!(
    Source=Copy Pose From Mesh() {bUseAttachedParent=true}
    Alpha
    NotCrouch:double=Get NotCrouching() -> (NotCrouching:double)
    CrouchStatus:double=Get CrouchAlpha() -> (CrouchAlpha:double)
    movingCtrl:double=Get SprintCntrolRig() -> (SprintCntrolRig:double)
    PistolMove:double=Get AlphaPistol() -> (AlphaPistol:float)
   ) -> (Pose) {
    alphaBoolBlend {blendOption=Linear}
    controlRigAssetReference {blueprintRigClass=&/F_MED_StrategicSpur_BlastHelmet_CtrlRig.F_MED_StrategicSpur_BlastHelmet_CtrlRig_C}
   }
   Save cached pose 'CRCache_Input'(Pose=@n2) {cacheName=CRCache_Input}
   n3 = Anim Dynamics - Bone: dyn_helmetHair_01_m!(
    ComponentPose=Local To Component(LocalPose=Use cached pose 'CRCache_Input'())
    Alpha=1
    GravityOverride:Vector="0,0,-600"
    ExternalForce="0,0,0"
   ) -> (Pose) {
    angularDampingOverride=1
    angularSpringConstant=20000
    bAngularSpring=true
    bOverrideAngularDamping=true
    bUseGravityOverride=true
    gravityScale=0.35
    simulationSpace=RootRelative
    alphaBoolBlend {blendOption=Linear}
    boundBone {boneName=dyn_helmetHair_01_m}
    componentAppliedLinearAccClamp=<5000 5000 5000>
    componentLinearAccScale=<0.4 0.4 0.4>
    componentLinearVelScale=<0.4 0.4 0.4>
    gravityOverride {z=-600}
    physicsBodyDefinitions [{
      boundBone {boneName=dyn_helmetHair_01_m}
      boxExtents=<20 30 30>
      constraintSetup {angularLimitsMax=<0 20 35>, angularLimitsMin=<0 -25 -35>, angularTarget=<1 0 0>}
      localJointOffset=<3 0 0>
     }]
   }
   n4 = Anim Dynamics - Bone: dyn_helmetHair_02_m!(ComponentPose=@n3, Alpha=1, GravityOverride:Vector="0,0,-600", ExternalForce="0,0,0") -> (Pose) {
    angularDampingOverride=1
    angularSpringConstant=20000
    bAngularSpring=true
    bOverrideAngularDamping=true
    bUseGravityOverride=true
    gravityScale=0.35
    simulationSpace=RootRelative
    alphaBoolBlend {blendOption=Linear}
    boundBone {boneName=dyn_helmetHair_02_m}
    componentAppliedLinearAccClamp=<5000 5000 5000>
    componentLinearAccScale=<0.3 0.3 0.3>
    componentLinearVelScale=<0.3 0.3 0.3>
    gravityOverride {z=-600}
    physicsBodyDefinitions [{
      boundBone {boneName=dyn_helmetHair_02_m}
      boxExtents=<20 30 30>
      constraintSetup {angularLimitsMax=<0 20 35>, angularLimitsMin=<0 -25 -35>, angularTarget=<1 0 0>}
      localJointOffset=<3 0 0>
     }]
   }
   n5 = Anim Dynamics - Bone: dyn_helmetHair_03_m!(ComponentPose=@n4, Alpha=1, GravityOverride:Vector="0,0,-600", ExternalForce="0,0,0") -> (Pose) {
    angularDampingOverride=1
    angularSpringConstant=20000
    bAngularSpring=true
    bOverrideAngularDamping=true
    bUseGravityOverride=true
    gravityScale=0.35
    simulationSpace=RootRelative
    alphaBoolBlend {blendOption=Linear}
    boundBone {boneName=dyn_helmetHair_03_m}
    componentAppliedLinearAccClamp=<5000 5000 5000>
    componentLinearAccScale=<0.4 0.4 0.4>
    componentLinearVelScale=<0.4 0.4 0.4>
    gravityOverride {z=-600}
    physicsBodyDefinitions [{
      boundBone {boneName=dyn_helmetHair_03_m}
      boxExtents=<20 30 30>
      constraintSetup {angularLimitsMax=<0 20 35>, angularLimitsMin=<0 -20 -35>, angularTarget=<1 0 0>}
      localJointOffset=<3 0 0>
     }]
   }
   n6 = Anim Dynamics - Bone: dyn_helmetHair_04_m!(ComponentPose=@n5, Alpha=1, GravityOverride:Vector="0,0,-650", ExternalForce="0,0,0") -> (Pose) {
    angularDampingOverride=0.9999
    angularSpringConstant=13000
    bAngularSpring=true
    bOverrideAngularDamping=true
    bUseGravityOverride=true
    gravityScale=0.35
    simulationSpace=RootRelative
    alphaBoolBlend {blendOption=Linear}
    boundBone {boneName=dyn_helmetHair_04_m}
    componentAppliedLinearAccClamp=<5000 5000 5000>
    componentLinearAccScale=<0.8 0.8 0.8>
    componentLinearVelScale=<0.8 0.8 0.8>
    gravityOverride {z=-650}
    physicsBodyDefinitions [{
      boundBone {boneName=dyn_helmetHair_04_m}
      boxExtents=<20 30 30>
      constraintSetup {angularLimitsMax=<0 30 50>, angularLimitsMin=<0 -30 -50>, angularTarget=<1 0 0>}
      localJointOffset=<3 0 0>
     }]
   }
   n7 = Anim Dynamics - Bone: dyn_helmetHair_05_m!(ComponentPose=@n6, Alpha=1, GravityOverride:Vector="0,0,-650", ExternalForce="0,0,0") -> (Pose) {
    angularDampingOverride=0.9999
    angularSpringConstant=13000
    bAngularSpring=true
    bOverrideAngularDamping=true
    bUseGravityOverride=true
    gravityScale=0.35
    simulationSpace=RootRelative
    alphaBoolBlend {blendOption=Linear}
    boundBone {boneName=dyn_helmetHair_05_m}
    componentAppliedLinearAccClamp=<5000 5000 5000>
    componentLinearAccScale=<0.8 0.8 0.8>
    componentLinearVelScale=<0.8 0.8 0.8>
    gravityOverride {z=-650}
    physicsBodyDefinitions [{
      boundBone {boneName=dyn_helmetHair_05_m}
      boxExtents=<20 30 30>
      constraintSetup {angularLimitsMax=<0 30 50>, angularLimitsMin=<0 -30 -50>, angularTarget=<1 0 0>}
      localJointOffset=<3 0 0>
     }]
   }
   n8 = Anim Dynamics - Bone: dyn_helmetHair_06_m!(ComponentPose=@n7, Alpha=0.5, GravityOverride:Vector="0,0,-650", ExternalForce="0,0,0") -> (Pose) {
    angularDampingOverride=0.9999
    angularSpringConstant=8000
    bAngularSpring=true
    bOverrideAngularDamping=true
    bUseGravityOverride=true
    gravityScale=0.35
    simulationSpace=RootRelative
    alphaBoolBlend {blendOption=Linear}
    boundBone {boneName=dyn_helmetHair_06_m}
    componentAppliedLinearAccClamp=<5000 5000 5000>
    componentLinearAccScale=<0.4 0.4 0.4>
    componentLinearVelScale=<0.4 0.4 0.4>
    gravityOverride {z=-650}
    physicsBodyDefinitions [{
      boundBone {boneName=dyn_helmetHair_06_m}
      boxExtents=<20 30 30>
      constraintSetup {angularLimitsMax=<0 0 50>, angularLimitsMin=<0 -35 -50>, angularTarget=<1 0 0>}
      localJointOffset=<3 0 0>
     }]
   }
   n9 = Anim Dynamics - Bone: dyn_helmetHairB_06_m!(ComponentPose=@n8, Alpha=1, GravityOverride:Vector="0,0,-650", ExternalForce="0,0,0") -> (Pose) {
    angularDampingOverride=0.9999
    angularSpringConstant=13000
    bAngularSpring=true
    bOverrideAngularDamping=true
    bUseGravityOverride=true
    gravityScale=0.35
    simulationSpace=RootRelative
    alphaBoolBlend {blendOption=Linear}
    boundBone {boneName=dyn_helmetHairB_06_m}
    componentLinearAccScale=<1 1 1>
    componentLinearVelScale=<1 1 1>
    gravityOverride {z=-650}
    physicsBodyDefinitions [{
      boundBone {boneName=dyn_helmetHairB_06_m}
      boxExtents=<20 30 30>
      constraintSetup {angularLimitsMax=<0 0 30>, angularLimitsMin=<0 0 -20>, angularTarget=<1 0 0>}
      localJointOffset=<3 0 0>
     }]
   }
   Save cached pose 'CRCache_Dyn'(Pose=Component To Local(ComponentPose=@n9)) {cacheName=CRCache_Dyn}
  ]}
 EventGraph {nodes [
   Event BlueprintUpdateAnimation()
   > n1 = Sequence!(execute) -> (then_0, then_1, then_2:Exec, then_3:Exec, then_4:Exec)
   Set CrouchAlpha(execute=@n1.then_1, CrouchAlpha:double="To Float (Boolean)"(InBool=Get bIsCrouching())) -> (Output_Get:double)
   n2 = Get bIsCrouching()
   > Set NotCrouching(NotCrouching:double="To Float (Boolean)"(InBool=NOT Boolean(A=@n2))) -> (Output_Get:double)
   n3 = Get OnZipline() -> (OnZipline:bool)
   > Set ZiplineAlpha(ZiplineAlpha:double="To Float (Boolean)"(InBool=NOT Boolean(A=@n3))) -> (Output_Get:double)
   n4 = Gravity Override State Function DEPRECATED(
    execute=@n1.then_0
    CharacterPartAnimInstance=Self-Reference()
    GameTimeInSeconds=GetGameTimeInSeconds()
    GravityOverrideParams=Get Gravity Override Params()
   )
   > Set GravityOverride(GravityOverride=@n4)
   n5 = Branch(execute=@n1.then_3, Condition=OR Boolean(A=Get bIsMoving2D(), B=Get bIsSkydiveFloating()))
   > Set SprintCntrolRig(SprintCntrolRig:double=0.5) -> (Output_Get:double)
   Set SprintCntrolRig(execute=@n5.else, SprintCntrolRig=1)
   Sequence!(execute=@n1.then_4) -> (then_0, then_1)
   > n6 = Cast To FortPlayerAnimInstance(Object=GetAnimInstance(self=GetOwnerSkeletalMeshComponent()))
   > Set AlphaPistol(AlphaPistol:float=0) -> (Output_Get:float)
   > n7 = Switch on EFortWeaponCoreAnimation!(execute, Selection:EFortWeaponCoreAnimation Enum=Get WeaponCoreAnim(self=@n6)) -> (
    Melee:Exec
    Pistol:Exec
    Shotgun:Exec
    PaperBlueprint:Exec
    Rifle:Exec
    MeleeOneHand:Exec
    MachinePistol:Exec
    RocketLauncher:Exec
    GrenadeLauncher:Exec
    GoingCommando:Exec
    AssaultRifle:Exec
    TacticalShotgun:Exec
    SniperRifle:Exec
    TrapPlacement:Exec
    ShoulderLauncher:Exec
    AbilityDecoTool:Exec
    Crossbow:Exec
    C4:Exec
    RemoteControl:Exec
    DualWield:Exec
    AR_BullPup:Exec
    AR_ForwardGrip:Exec
    MedPackPaddles:Exec
    SMG_P90:Exec
    AR_DrumGun:Exec
    Consumable_Small:Exec
    Consumable_Large:Exec
    Balloon:Exec
    MountedTurret:Exec
    CreativeTool:Exec
    ExplosiveBow:Exec
    AshtonIndigo:Exec
    AshtonChicago:Exec
    MeleeDualWield:Exec
    Unarmed:Exec
   )
   Set AlphaPistol(execute=[@n7.DualWield, @n7.Pistol, @n7.RocketLauncher], AlphaPistol=1)
   n8 = Branch(execute=@n1.then_2, Condition=Get bIsSkydiveFloating())
   > Set FloatingDynamics(FloatingDynamics:double=0.5) -> (Output_Get:double)
   Set FloatingDynamics(execute=@n8.else, FloatingDynamics=1)
   n9 = Event BlueprintProcessAnimInputEvent()
   > n10 = Branch(Condition=IsNamed(self=@n9.AnimInputEvent, InEventName=aie.Player.Zipline.Entered))
   > Set OnZipline(OnZipline:bool=1) -> (Output_Get:bool)
   Branch(execute=@n10.else, Condition=IsNamed(self=@n9.AnimInputEvent, InEventName=aie.Player.Zipline.Exited))
   > Set OnZipline(OnZipline=false)
  ]}
]
variables [
 AlphaPistol {type=number}
 CrouchAlpha {type=number}
 FloatingDynamics {type=number}
 Gravity Override Params {type=GravityOverrideParamsStruct}
 GravityOverride {type=Vector}
 NotCrouching {type=number, default=1}
 OnZipline {type=boolean}
 SprintCntrolRig {type=number}
 ZiplineAlpha {type=number}
]
